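# Default Lit mesh-particle material for a toy ball marker effect that spins, bobs and glows.
# Time-driven rotation about world Z around the particle position combines with a sine bounce on Z, scaled by a dynamic parameter and passed through MF_MakePrecise.
# Emissive adds a dim particle-color base to a boosted sphere-mask glow centered on the bouncing particle position, then runs through MF_FX_EmissiveAdjustments.
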Material /FNGameplayCues/Effects/Fort_Effects/Materials/UI/M_Toy_Ball_Marker
graph {
 properties {bUsedWithMeshParticles=true, pixelDepthOffsetMode=PDOM_Legacy, usedShadingModels=Default Lit}
 nodes [
  n1 = Time()
  n2 = ScalarParameter() {defaultValue=1, parameterName=RotateSpeed}
  n3 = Time()
  n4 = ScalarParameter() {defaultValue=0.75, parameterName=BounceSpeed}
  n5 = Multiply(A=Sine(None=Multiply(A=@n3, B=@n4)), B=Constant3Vector() {constant=<rgba 0 0 1 0>})
  n6 = ScalarParameter() {defaultValue=6, parameterName=BounceAmount}
  n7 = MaterialFunctionCall!(InVector=Multiply(A=DynamicParameter() {paramNames [BonceAmount,Param2,Param3,Param4]}.BonceAmount, B=Multiply(A=@n5, B=@n6))) {materialFunction=/Game/Athena/Environments/Landscape/Foliage/MF_MakePrecise}
  Add(
   A=RotateAboutAxis(
    NormalizedRotationAxis=Constant3Vector() {constant=<rgba 0 0 1 0>}
    RotationAngle=Multiply(A=@n1, B=@n2)
    PivotPoint=ParticlePositionWS()
    Position=WorldPosition().XYZ
   )
   B=@n7.Result
  )
  n8 = ParticleColor()
  n9 = ScalarParameter() {defaultValue=34.7352, parameterName=EmissiveBoost}
  n10 = Multiply(A=@n8.RGB, B=@n9) {constB=0.5}
  n11 = WorldPosition()
  n12 = Add(A=ParticlePositionWS(), B=@n7.Result)
  n13 = ScalarParameter() {defaultValue=40, parameterName=GlowRadius}
  n14 = Constant()
  n15 = Multiply(
   A=@n8.RGB
   B=Power(Base=SphereMask(A=@n11.XYZ, B=@n12, Radius=@n13, Hardness=@n14), Exp=ScalarParameter() {defaultValue=3, parameterName=GlowPower})
  )
  MaterialFunctionCall(Emissive=Add(A=Multiply(A=@n10) {constB=0.05}, B=Multiply(A=@n15, B=@n9))) {materialFunction=&MF_FX_EmissiveAdjustments}
  ScalarParameter() {defaultValue=0.25, parameterName=Metalic}
  ScalarParameter() {defaultValue=1, parameterName=Spec}
  ScalarParameter() {defaultValue=0.5, parameterName=Roughness}
 ]
}
